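; Global manager for Military Rank that grants team rank points and spawns rank-up drop pods.
; On authority, it binds POI banner captures directly to avoid duplicate grants and listens for safe-zone phases to award circle-scaled storm points per team.
; Rank-up requests enter a queue drained by a randomly staggered timer, one EQS query at a time.
; Successful queries spawn and mark a pod; failed ones retry with a fallback query, then requeue.
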(event EventBeginPlay
  ;[0.15 1 0.15 0.5] Handle POI Banner capture manually (instead of using MilitaryRank_Reactions) to avoid duplicate point grants
  (Networking|SwitchHasAuthority
    (:Authority
      (Utilities|FlowControl|Sequence
        (:then_0
          (Class|FortAsyncActionCurrentPlaylistReady|PlaylistDataReadyAsync
            (:OnReady
              (Utilities|FlowControl|DelayUntilNextTick) ; PlaylistDataReadyAsync is when the banners spawn. Wait to avoid race condition.
              (bind _outactors (Actor|GetAllActorsOfClass "/POIBanner/Actors/BGA_POIBanner_Banner.BGA_POIBanner_Banner_C"))
              (bind _output_get (Variables|Default|SetPOI%20Banners _outactors))
              (Default|BindEventtoOnBannerRaised _output_get (EventDispatchers|CreateEvent :!event Custom|GrantSquadBannerPoints)))))
        (:then_1
          ;;[0.5 1 0 0.5] Storm Phase Handling
          (if (ScalableFloat|GetValueasBool (Variables|HF|GetHF_SurviveStormCircleCanGrant))
            (CallFunction|ListenforGlobalStormEvent)))))))

(event Custom|GrantSquadBannerPoints
  (params (Team :type Integer) (Winners :type Array%20of%20Fort%20Player%20Pawn%20Object%20References) (Banner :type BGA%20POIBanner%20Banner%20Object%20Reference))
  ;[0.15 1 0.15 0.5] Handle POI Banner capture manually (instead of using MilitaryRank_Reactions) to avoid duplicate point grants
  (CallFunction|AddHeroPointstoTeam (Teams|IndextoTeam Team) (ScalableFloat|GetValueatLevel (Variables|HF|GetHF_POIBannerRankPoints)) (Variables|Default|GetContextTag)))

(event EventEndPlay
  (params (EndPlayReason :type EEndPlayReason%20Enum :default "Destroyed"))
  (bind _safezoneeventlisteningtask (Variables|Default|GetSafeZoneEventListeningTask))
  ;[0.15 1 0.15 0.5] Handle POI Banner capture manually (instead of using MilitaryRank_Reactions) to avoid duplicate point grants
  (Default|UnbindEventfromOnBannerRaised (Variables|Default|GetPOI%20Banners) (EventDispatchers|CreateEvent :!event Custom|GrantSquadBannerPoints))
  ;;[0.5 1 0 0.5] Storm Phase Handling
  (Utilities|IsValid _safezoneeventlisteningtask
    (:Is%20Valid
      (GameplayEventRouter|StopListeningtoStatefulEvent _safezoneeventlisteningtask))))

(event EventSpawnDropPodForRankUp
  (params (PlayerPawn :type Fort%20Player%20Pawn%20Athena%20Object%20Reference) (NewRank :type Gameplay%20Tag%20Structure))
  (if (> (ScalableFloat|GetValueatLevel (Variables|HF|GetHF_DropPod_Enabled)) 0)
    (if (ScalableFloat|GetValueasBool (Variables|HF|GetHF_BlockDropForBots))
      (if (Pawn|IsBotControlled PlayerPawn)
        _
        (else
          _)))
    (Utilities|Array|Add (Variables|DropPodSpawnManagement|GetDropPodRequestQueue) (Utilities|Struct|MakeMilitaryRankDropPodSpawnRequest PlayerPawn NewRank))
    (CallFunction|SetStaggeredDropPodSpawnTimer)))

(event Custom|SetStaggeredDropPodSpawnTimer
  (Utilities|Time|ClearandInvalidateTimerbyHandle (Variables|DropPodSpawnManagement|GetStaggerTimer))
  (bind _returnvalue (Utilities|Time|SetTimerbyEvent (delegate Custom|OnSpawnTimerFinished) (ScalableFloat|GetValueatLevel (Variables|HF|GetHF_DropPod_Delay_Initial)) false false (Math|Float|Max(Float) (+ (ScalableFloat|GetValueatLevel (Variables|HF|GetHF_DropPod_Delay_BetweenSpawns)) (Math|Random|RandomFloatinRange 0.0 (ScalableFloat|GetValueatLevel (Variables|HF|GetHF_DropPod_Delay_Variance)))))))
  (Variables|DropPodSpawnManagement|SetStaggerTimer _returnvalue))

(event Custom|OnSpawnTimerFinished
  (CallFunction|TrySpawnDropPod)
  (CallFunction|SetStaggeredDropPodSpawnTimer))

(event Custom|ListenForGlobalStormEvent
  ; The Storm Survived event is a global event so we want to make sure each squad only gets it once
  (CallFunction|CacheScalableFloats)
  (bind (_asynctaskproxy _payload) (GameplayEventRouter|StartListeningToStatefulEvent (GameplayEventRouter|GetGlobalEventRouter) "/Script/FortniteGame.FortSafeZonePhaseUpdatedEvent")
    (:then
      (Variables|Default|SetSafeZoneEventListeningTask _asynctaskproxy))
    (:OnEventReceived
      _)
    (:OnSavedEventStateExists
      _))
  ;;[0.8229 0.8229 0.8229 0.5] We skip first circle formed, which counts as a 'circle closed event'
  (if (> (Utilities|Struct|BreakFortSafeZonePhaseUpdatedEvent _payload) 1)
    (bind _survivestormcirclepoints (CallFunction|AdjustPointsBasedonCircle (Utilities|Struct|BreakFortSafeZonePhaseUpdatedEvent _payload)))
    (bind _teamsset1 (CallFunction|GetTeamsLeft))
    (bind _valuepin (Utilities|Set|ForEachLoop(Set) _teamsset1))
    (CallFunction|AddHeroPointstoTeam _valuepin _survivestormcirclepoints (GameplayTags|MakeLiteralGameplayTag "(TagName=\"CherrySmoke.HeroRank.Points.Context.SurviveStormPhase\")"))))

(fn AdjustPointsBasedOnCircle
  (params (Safe%20Zone%20Phase%20Index :type Integer))
  (if (and (>= Safe%20Zone%20Phase%20Index (Variables|Default|GetSurviveStormCirclePointsIncreaseMinCircle)) (<= Safe%20Zone%20Phase%20Index (Variables|Default|GetSurviveStormCirclePointsIncreaseMaxCircle)))
    (bind _output_get (Variables|Default|SetSurviveStormCirclePoints (+ (Variables|Default|GetSurviveStormCirclePoints) (Variables|Default|GetSurviveStormCirclePointsIncrease))))
    (return _output_get)
    (else
      (return (Variables|Default|GetSurviveStormCirclePoints)))))

(fn CacheScalableFloats
  (Variables|Default|SetSurviveStormCirclePoints (ScalableFloat|GetValueatLevel (Variables|HF|GetHF_SurviveStormCirclePoints)))
  (Variables|Default|SetSurviveStormCirclePointsIncrease (ScalableFloat|GetValueasInteger (Variables|HF|GetHF_SurviveStormCirclePointsIncrease)))
  (Variables|Default|SetSurviveStormCirclePointsIncreaseMinCircle (ScalableFloat|GetValueasInteger (Variables|HF|GetHF_SurviveStormCirclePointsIncreaseMinCircle)))
  (Variables|Default|SetSurviveStormCirclePointsIncreaseMaxCircle (ScalableFloat|GetValueasInteger (Variables|HF|GetHF_SurviveStormCirclePointsIncreaseMaxCircle))))

(fn FinishSpawnDropPod
  (params (QueryInstance :type EQS%20Query%20Instance%20Object%20Reference) (QueryStatus :type EEnvQueryStatus%20Enum :default "Processing"))
  (bind _pendingspawnrequest (Variables|DropPodSpawnManagement|GetPendingSpawnRequest))
  (bind (_playerpawn_6_6879a82141a04f00f07c52be50b3edd3 _rank_5_5acf6bc540e63870c68d5681104db714 _usefallbackquery_9_186df4c54ec095798be4daa78b1454ac) (Utilities|Struct|BreakMilitaryRankDropPodSpawnRequest _pendingspawnrequest))
  (bind _pendingspawnrequest_2 (Variables|DropPodSpawnManagement|GetPendingSpawnRequest))
  (bind (_playerpawn_6_6879a82141a04f00f07c52be50b3edd3_2 _rank_5_5acf6bc540e63870c68d5681104db714_2) (Utilities|Struct|BreakMilitaryRankDropPodSpawnRequest _pendingspawnrequest_2))
  (bind _map_ranktoint (Variables|DropPodSpawnManagement|GetMap_RankToInt))
  (bind (_value _returnvalue) (Utilities|Map|Find _map_ranktoint _rank_5_5acf6bc540e63870c68d5681104db714_2))
  (bind (_returnvalue_roll _returnvalue_pitch _returnvalue_yaw) (Math|Rotator|BreakRotator (Transformation|GetActorRotation (Utilities|Struct|BreakMilitaryRankDropPodSpawnRequest (Variables|DropPodSpawnManagement|GetPendingSpawnRequest)))))
  (if (Utilities|Enum|Equal(Enum) QueryStatus "Success")
    (Utilities|FlowControl|Sequence
      (:then_0
        (Utilities|IsValid _playerpawn_6_6879a82141a04f00f07c52be50b3edd3_2
          (:Is%20Valid
            (if _returnvalue
              (LocalVariables|SetRankAsInt _value)
              (bind (_resultlocations _returnvalue_2) (AI|EQS|GetQueryResultsasLocations (Variables|GetQueryInstance)))
              (if (and _returnvalue_2 (Utilities|Array|IsNotEmpty _resultlocations))
                (bind _returnvalue_3 (Game|SpawnActorfromClass "/MilitaryRank/Gameplay/BP_MilitaryRank_DropPod.BP_MilitaryRank_DropPod_C" (Math|Transform|MakeTransform (Utilities|Array|Get(acopy) _resultlocations) (Math|Rotator|MakeRotator 0.0 0.0 (+ _returnvalue_yaw (Math|Float|SelectFloat 9e+01 -9e+01 (>= (LocalVariables|GetRankAsInt) 3))))) "Undefined" "MultiplyWithRoot" 0 0 0 0))
                (bind _output_get (LocalVariables|SetSpawned _returnvalue_3))
                (Class|BPMilitaryRankDropPod|ServerSetandRepDropPodRank _output_get (LocalVariables|GetRankAsInt))
                (MarkerComponent|MarkActoronServer (Actor|GetComponentbyClass (Fortnite|GetFortPlayerControllerfromActor (Utilities|Struct|BreakMilitaryRankDropPodSpawnRequest (Variables|DropPodSpawnManagement|GetPendingSpawnRequest))) "/Script/FortniteGame.AthenaMarkerComponent") (LocalVariables|GetSpawned) false true))))))
      (:then_1
        _))
    (elif _usefallbackquery_9_186df4c54ec095798be4daa78b1454ac
      ;[0.15 0.09928 0.0987 0.5] Handle EQS failure
      ;;[0.08247 0.02592 0.15 0.5] Fallback failed. Back of the line and try again!
      (Utilities|Array|Add (Variables|DropPodSpawnManagement|GetDropPodRequestQueue) (Utilities|Struct|MakeMilitaryRankDropPodSpawnRequest _playerpawn_6_6879a82141a04f00f07c52be50b3edd3 _rank_5_5acf6bc540e63870c68d5681104db714))
      (else
        ;;[1 0.2582 0 0.5] Ideal query failed. Immediately retry with fallback query.
        (Variables|DropPodSpawnManagement|SetPendingSpawnRequest (Utilities|Struct|MakeMilitaryRankDropPodSpawnRequest _playerpawn_6_6879a82141a04f00f07c52be50b3edd3 _rank_5_5acf6bc540e63870c68d5681104db714 true))
        (bind _returnvalue_4 (AI|EQS|RunEQSQuery "/MilitaryRank/Gameplay/EQS_DropPodSpawnLocation_Fallback.EQS_DropPodSpawnLocation_Fallback" _playerpawn_6_6879a82141a04f00f07c52be50b3edd3 "RandomBest25Pct"))
        (AI|EQS|BindEventtoOnQueryFinished _returnvalue_4 (EventDispatchers|CreateEvent :!event FinishSpawnDropPod)))))
  ; Clear out pending spawn request
  (Variables|DropPodSpawnManagement|SetPendingSpawnRequest (Utilities|Struct|MakeMilitaryRankDropPodSpawnRequest 0)))

(fn GetTeamsLeft
  (for _array_element (Variables|/Script/Engine.GameStateBase|GetPlayerArray (Fortnite|Athena|GameState|GetGameStateAthena))
    (bind _asfort_player_state_athena (Utilities|Casting|CastToFortPlayerStateAthena _array_element))
    (Utilities|Set|Add (LocalVariables|GetTeamsLeft) (Class|FortPlayerStateAthena|GetTeam _asfort_player_state_athena)))
  (return (LocalVariables|GetTeamsLeft)))

(fn TrySpawnDropPod
  (bind (_playerpawn_6_6879a82141a04f00f07c52be50b3edd3 _rank_5_5acf6bc540e63870c68d5681104db714) (Utilities|Struct|BreakMilitaryRankDropPodSpawnRequest (Variables|DropPodSpawnManagement|GetPendingSpawnRequest)))
  (bind _map_ranktoint (Variables|DropPodSpawnManagement|GetMap_RankToInt))
  ; Early out if we're working on a spawn req already
  (if (not (GameplayTags|IsGameplayTagValid _rank_5_5acf6bc540e63870c68d5681104db714))
    (if (Utilities|Array|IsNotEmpty (Variables|DropPodSpawnManagement|GetDropPodRequestQueue))
      (bind _output_get (LocalVariables|SetCandidate (Utilities|Array|Get(acopy) (Variables|DropPodSpawnManagement|GetDropPodRequestQueue))))
      (Utilities|Array|RemoveIndex (Variables|DropPodSpawnManagement|GetDropPodRequestQueue))
      (Utilities|IsValid (Utilities|Struct|BreakMilitaryRankDropPodSpawnRequest _output_get)
        (:Is%20Valid
          (if (.ReturnValue (Utilities|Map|Find _map_ranktoint (.Rank_5_5ACF6BC540E63870C68D5681104DB714 (Utilities|Struct|BreakMilitaryRankDropPodSpawnRequest _output_get))))
            (LocalVariables|SetRankAsInt (Utilities|Map|Find _map_ranktoint (.Rank_5_5ACF6BC540E63870C68D5681104DB714 (Utilities|Struct|BreakMilitaryRankDropPodSpawnRequest _output_get))))
            (bind _returnvalue (AI|EQS|RunEQSQuery "/MilitaryRank/Gameplay/EQS_DropPodSpawnLocation.EQS_DropPodSpawnLocation" (Utilities|Struct|BreakMilitaryRankDropPodSpawnRequest _output_get) "RandomBest25Pct"))
            (AI|EQS|BindEventtoOnQueryFinished _returnvalue (EventDispatchers|CreateEvent :!event FinishSpawnDropPod))
            (Variables|DropPodSpawnManagement|SetPendingSpawnRequest (LocalVariables|GetCandidate))))))))
