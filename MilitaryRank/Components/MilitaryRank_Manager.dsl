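; Military Rank manager component on the player state that syncs hero rank tags and hands out rank-up rewards.
; A rank change updates loose tags, fires a burst cue at rank A, requests a drop pod through the global ranking manager's spawner interface, and sends a quest event.
; At the bunker-unlock rank, a hotfix-gated timer marks bunker doors in a specific state.
; It handles ToyFox pilots specially, adds a cinematic pod cheat, and stubs chest pinging.
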(event EventOnRuntimeDataCreated
  (CallFunction|UpdateRankTag)
  (Development|PrintString "Military Rank Manager - Runtime Data Created" (Variables|Default|GetDebug) (Variables|Default|GetDebug) "(R=1.000000,G=0.000000,B=0.000000,A=1.000000)" 5.0))

(event EventBeginPlay
  (CallFunction|UpdateRankTag)
  (Development|PrintString "Military Rank Manager - Begin Play" (Variables|Default|GetDebug) (Variables|Default|GetDebug) "(R=1.000000,G=0.000000,B=0.000000,A=1.000000)" 5.0))

(event EventOnRankChanged
  (params (NewRank :type Gameplay%20Tag%20Structure%20%28by%20ref%29))
  (CallFunction|UpdateRankTag)
  (Utilities|FlowControl|Sequence
    (:then_0
      (CallFunction|StartGiveRankRewards)
      (CallFunction|SendQuestEventonRankUp))
    (:then_1
      (Development|PrintString "Military Rank Manager - Skipping EQS query for closest discoverable" (Variables|Default|GetDebug) (Variables|Default|GetDebug) "(R=1.000000,G=0.000000,B=0.000000,A=1.000000)" 5.0))))

(event Custom|OnHeroRankDisoverableEQSFinished
  (params (QueryInstance :type EQS%20Query%20Instance%20Object%20Reference) (QueryStatus :type EEnvQueryStatus%20Enum :default "Processing"))
  ;[1 0 0 0.5] Mark the Closest Chest
  (Development|PrintString "Military Rank Manager - Skipping EQS finished" (Variables|Default|GetDebug) (Variables|Default|GetDebug) "(R=1.000000,G=0.000000,B=0.000000,A=1.000000)" 5.0))

(event Custom|PingChestDelay
  ;[1 0 0 0.5] Mark the Closest Chest
  (Development|PrintString "Military Rank Manager - Skipping ping chest delay" (Variables|Default|GetDebug) (Variables|Default|GetDebug) "(R=1.000000,G=0.000000,B=0.000000,A=1.000000)" 5.0))

(event Custom|CheatSpawnDropPod
  (params (RankAsInt :type Integer) (DistanceInFrontOfPlayer :type Float%20%28double-precision%29))
  ; Cheats for cinematics - use KE* CheatSpawnDropPod <rank#> <distanceInFrontOfPlayer#>
  (Networking|SwitchHasAuthority
    (:Remote
      (CallFunction|RepCheatSpawnDropPod RankAsInt DistanceInFrontOfPlayer))))

(event EventOnHeroPointsChanged
  (params (CurrentPoints :type Float%20%28single-precision%29))
  ;[1 0 0 0.5] Remove Print
  (Development|PrintString (Utilities|String|Append "Military Rank Manager - Current Points Changed to: " (Utilities|String|ToString(Float) CurrentPoints)) (Variables|Default|GetDebug) (Variables|Default|GetDebug) "(R=1.000000,G=0.000000,B=0.000000,A=1.000000)" 5.0))

(event Custom|RepCheatSpawnDropPod
  (params (RankAsInt :type Integer) (DistanceInFrontOfPlayer :type Float%20%28double-precision%29))
  ; Cheats for cinematics - use KE* CheatSpawnDropPod <rank#> <distanceInFrontOfPlayer#>
  (Networking|SwitchHasAuthority
    (:Authority
      (bind _asfort_player_state (Utilities|Casting|CastToFortPlayerState (Components|GetOwner)))
      (bind _asfort_player_pawn_athena (Utilities|Casting|CastToFortPlayerPawnAthena (PlayerState|GetPawn _asfort_player_state)))
      (Variables|Default|SetPlayerPawn _asfort_player_pawn_athena)
      (Default|SpawnDropPodforCheatsOnly (Variables|Default|GetPlayerPawn) RankAsInt DistanceInFrontOfPlayer))))

(event Custom|SpawnDropPod
  (bind _playerpawn (Variables|Default|GetPlayerPawn))
  (Networking|SwitchHasAuthority
    (:Authority
      ;[0.8 0.8 0.15 0.5] Are we ToyFox? If so, don't grant until we exit. Need to wait a frame and reset the playerpawn ref here so we don't try to spawn from ToyFox as its being cleaned up
      (if (GameplayTags|HasMatchingGameplayTag _playerpawn "(TagName=\"ToyFox.Player.Pilot\")")
        (Ability|Tasks|WaitGameplayTagRemoveFromActor _playerpawn "(TagName=\"ToyFox.Player.Pilot\")" true
          (:Removed
            (Utilities|FlowControl|DelayUntilNextTick)
            (bind _asfort_player_state (Utilities|Casting|CastToFortPlayerState (Components|GetOwner)))
            (bind _asfort_player_pawn_athena (Utilities|Casting|CastToFortPlayerPawnAthena (PlayerState|GetPawn _asfort_player_state)))
            (Variables|Default|SetPlayerPawn _asfort_player_pawn_athena))))
      (bind _returnvalue (Class|CherrySmokeHeroRankingGlobalManager|GetHeroRankingGlobalManager))
      (bind _asbpi_drop_pod_spawner (Utilities|Casting|CastToBPI_DropPodSpawner _returnvalue))
      (Class|BPIDropPodSpawner|SpawnDropPodforRankUp _asbpi_drop_pod_spawner (Variables|Default|GetPlayerPawn) (CallFunction|GetHeroRank)))))

(event Custom|QueuePingOpenableBunkers
  ; Queue Ping Openable Bunkers
  (if (> (ScalableFloat|GetValueatLevel (Variables|Hotfixes|GetHF_BunkerPing_Enabled)) 0)
    (Utilities|Time|SetTimerbyEvent (delegate Custom|PingOpenableBunkers) (ScalableFloat|GetValueatLevel (Variables|Hotfixes|GetHotfix_BunkerPingDelay)))))

(event Custom|PingOpenableBunkers
  ; Ping Openable Bunkers
  (bind _outactors (Actor|GetAllActorsOfClass "/MidMatchWeaponBunkerGameplay/Items/WeaponBunkerDoor/MMOBunkers_Ch6S4/BGA_BunkerDoor_TidalCrane.BGA_BunkerDoor_TidalCrane_C"))
  (for _array_element _outactors
    (if (Utilities|Enum|Equal(Enum) (Variables|/MidMatchWeaponBunkerGameplay/Items/WeaponBunkerDoor/MMOBunkers_Ch6S4/BGA_BunkerDoor_TidalCrane.BGA_BunkerDoor_TidalCrane_C|GetWeapon%20Bunker%20Door%20State _array_element) "NewEnumerator6")
      (MarkerComponent|MarkActoronServer (Actor|GetComponentbyClass (Fortnite|GetFortPlayerControllerfromActor (Components|GetOwner)) "/Script/FortniteGame.AthenaMarkerComponent") _array_element false true))))

(fn GetHeroRankingTagsFromOwner
  ;[1 0 0 0.5] Update Tags?
  (return (GameplayTags|Filter (GameplayTags|GetOwnedGameplayTags (Ability|GetAbilitySystemComponent (Components|GetOwner)) :!class /Script/GameplayTags.BlueprintGameplayTagLibrary) (GameplayTags|MakeLiteralGameplayTagContainer "(GameplayTags=((TagName=\"CherrySmoke.HeroRank.A\"),(TagName=\"CherrySmoke.HeroRank.B\"),(TagName=\"CherrySmoke.HeroRank.C\"),(TagName=\"CherrySmoke.HeroRank.S\"),(TagName=\"CherrySmoke.HeroRank.S+\")))"))))

(fn SendQuestEventOnRankUp
  (Development|PrintString "Military Rank Manager - Sending Quest Event on Rank Up" (Variables|Default|GetDebug) (Variables|Default|GetDebug) "(R=1.000000,G=0.000000,B=0.000000,A=1.000000)" 5.0)
  ;[1 0 0 0.5] Update
  (Ability|SendGameplayEventtoActor (Variables|Default|GetPlayerPawn) "(TagName=\"CherrySmoke.HeroRank.Event.RankUp\")" (Utilities|Struct|MakeGameplayEventData "(TagName=\"CherrySmoke.HeroRank.Event.RankUp\")" (Variables|Default|GetPlayerPawn))))

(fn SpawnDropPod_ForCheatsOnly
  (params (PlayerPawn :type Fort%20Player%20Pawn%20Object%20Reference) (Rank :type Integer) (DistanceInFrontOfPlayer :type Float%20%28double-precision%29))
  (bind (_returnvalue_location _returnvalue_rotation) (Math|Transform|BreakTransform (Transformation|GetActorTransform (Variables|Default|GetPlayerPawn))))
  (bind (_returnvalue_roll _returnvalue_pitch _returnvalue_yaw) (Math|Rotator|BreakRotator (Transformation|GetActorRotation (Variables|Default|GetPlayerPawn))))
  (LocalVariables|SetSpawnLocation (+ _returnvalue_location (Utilities|Operators|Multiply (Math|Rotator|GetRotationXVector _returnvalue_rotation) (Math|Vector|MakeVector (Variables|GetDistanceInFrontOfPlayer) (Variables|GetDistanceInFrontOfPlayer)))))
  (bind (_outhit _returnvalue) (Collision|LineTraceByChannel (Utilities|Operators|Add (LocalVariables|GetSpawnLocation) (Math|Vector|MakeVector 0.0 0.0 (ScalableFloat|GetValueAtLevel(single-precision) (Variables|Hotfixes|GetHF_DropPod_StartingHeight)))) (Utilities|Operators|Subtract (LocalVariables|GetSpawnLocation) (Math|Vector|MakeVector 0.0 0.0 (ScalableFloat|GetValueAtLevel(single-precision) (Variables|Hotfixes|GetHF_DropPod_StartingHeight)))) "TraceTypeQuery10"))
  (if _returnvalue
    (LocalVariables|SetSpawnLocation (.Location (Collision|BreakHitResult _outhit)))
    (else
      _))
  (bind _returnvalue_2 (Game|SpawnActorfromClass "/MilitaryRank/Gameplay/BP_MilitaryRank_DropPod.BP_MilitaryRank_DropPod_C" (Math|Transform|MakeTransform (LocalVariables|GetSpawnLocation) (Math|Rotator|MakeRotator 0.0 0.0 (+ _returnvalue_yaw (Math|Float|SelectFloat 9e+01 -9e+01 (>= (Variables|GetRank) 3))))) "Undefined" "MultiplyWithRoot" 0 0 0 0))
  (bind _output_get (LocalVariables|SetSpawnedDropPod _returnvalue_2))
  (Class|BPMilitaryRankDropPod|ServerSetandRepDropPodRank _output_get (Variables|GetRank))
  (MarkerComponent|MarkActoronServer (Actor|GetComponentbyClass (Fortnite|GetFortPlayerControllerfromActor (Variables|Default|GetPlayerPawn)) "/Script/FortniteGame.AthenaMarkerComponent") (LocalVariables|GetSpawnedDropPod) false true))

(fn StartGiveRankRewards
  (Development|PrintString "Military Rank Manager - Start Give Rewards" (Variables|Default|GetDebug) (Variables|Default|GetDebug) "(R=1.000000,G=0.000000,B=0.000000,A=1.000000)" 5.0)
  (bind _asfort_player_state (Utilities|Casting|CastToFortPlayerState (Components|GetOwner)))
  (bind _asfort_player_pawn_athena (Utilities|Casting|CastToFortPlayerPawnAthena (PlayerState|GetPawn _asfort_player_state)))
  (Variables|Default|SetPlayerPawn _asfort_player_pawn_athena)
  (CallFunction|StartRankUpCosmetics)
  (CallFunction|SpawnDropPod)
  (if (== (CallFunction|GetHeroRank) (Variables|Default|GetRankRequiredToOpenBunker))
    (CallFunction|QueuePingOpenableBunkers)))

(fn StartRankUpCosmetics
  (Development|PrintString "Military Rank Manager - Start Rank Up Cosmetics - Set Rank FX Value" (Variables|Default|GetDebug) (Variables|Default|GetDebug) "(R=1.000000,G=0.000000,B=0.000000,A=1.000000)" 5.0)
  (switch Utilities|FlowControl|Switch|SwitchonGameplayTag (CallFunction|GetHeroRank)
    (:Default
      _)
    (:CherrySmoke.HeroRank.C
      _)
    (:CherrySmoke.HeroRank.B
      _)
    (:CherrySmoke.HeroRank.A
      (Variables|Default|SetCurrentRankFX 1.0)
      (GameplayCue|ExecuteGameplayCue%28Burst%29Local (Variables|Default|GetPlayerPawn) (Variables|GameplayCues|GetRank%20Up%20Burst%20Cue%20Tag) (Ability|GameplayCue|MakeGameplayCueParameters :NormalizedMagnitude 0.0 :RawMagnitude (Variables|Default|GetCurrentRankFX) :Location "0,0,0" :Normal "0,0,0" :Instigator (Variables|Default|GetPlayerPawn) :EffectCauser (Variables|Default|GetPlayerPawn) :GameplayEffectLevel 1 :AbilityLevel 1 :bReplicateLocationWhenUsingMinimalRepProxy false)))
    (:CherrySmoke.HeroRank.S
      (Variables|Default|SetCurrentRankFX 2.0))
    (:CherrySmoke.HeroRank.S+
      (Variables|Default|SetCurrentRankFX 3.0)))
  (Variables|Default|SetCurrentRankFX 0.0))

(fn UpdateRankTag
  (bind _returnvalue (Components|GetOwner))
  ; Updating Loose Tags on Server and Client so that they match the newest rank
  (LocalVariables|SetOwnerHeroRankTags (CallFunction|GetHeroRankingTagsfromOwner))
  (LocalVariables|SetCurrentHeroRankTags (CallFunction|GetHeroRank))
  (if (not (GameplayTags|MatchesAnyTags (LocalVariables|GetCurrentHeroRankTags) (LocalVariables|GetOwnerHeroRankTags) true))
    (Ability|GameplayEffect|RemoveLooseGameplayTags _returnvalue (LocalVariables|GetOwnerHeroRankTags))
    (Ability|GameplayEffect|AddLooseGameplayTags _returnvalue (GameplayTags|MakeGameplayTagContainerfromTag (LocalVariables|GetCurrentHeroRankTags)))))
